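# Master character material for the HeavyRoar body style, adding an animated style-transition wipe with a glowing, sparking edge.
# Curve parameters or an MPC drive progress; noise-broken sphere masks on pre-skinned position shape the wipe.
# Panning, warped noise creates sparks, and the edge mask boosts hue-shifted emissive.
# Shared PreFX/PostFX functions bracket optional skin FX (panning emissive, cloth fuzz, thin film, metal LUT) and boosted subsurface-profile skin.
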Material /BRCosmetics/Characters/Player/Male/Medium/Bodies/M_MED_HeavyRoar_Style/Materials/M_HeavyRoar_MASTER
graph {
 properties {
  bAutomaticallySetUsageInEditor=false
  bUseMaterialAttributes=true
  bUsedWithMeshDeformer=true
  bUsedWithMorphTargets=true
  bUsedWithNanite=true
  bUsedWithSkeletalMesh=true
  pixelDepthOffsetMode=PDOM_Legacy
  shadingModel=MSM_FromMaterialExpression
  subsurfaceProfile=&SS_HeroSkin_02
  usedShadingModels=Default Lit | Subsurface Profile | Cloth
 }
 nodes [
  n1 = Time()
  n2 = Constant2Vector() {g=-0.1, r=-0.05}
  n3 = TextureObjectParameter() {
   group=HeavyRoar
   parameterName=HR_TA
   samplerType=SAMPLERTYPE_Masks
   texture=/BRCosmetics/Characters/Player/Male/Medium/Bodies/M_MED_HeavyRoar_Style/Textures/T_M_MED_HeavyRoar_Style_TA
  }
  n4 = TextureSample(UVs=Add(A=Multiply(A=@n1, B=@n2) {constB=0.1}, B=TextureCoordinate()), Tex=@n3) {
   samplerType=SAMPLERTYPE_Masks
   texture=&T_M_MED_HeavyRoar_Style_TA
  }
  NamedRerouteDeclaration(None=@n4.B) {name=Perlin, nodeColor=<rgba 0 0.4375 1 1>}
  n5 = MaterialFunctionCall() {materialFunction=&BaseFN_CharacterFunction_PreFX}
  NamedRerouteDeclaration(None=@n5.Pre skinned Local Position) {name=LocalPosition, nodeColor=<rgba 1 0.4688 0 1>}
  n6 = NamedRerouteUsage()
  n7 = VectorParameter() {group=HeavyRoar, parameterName=HR_Wipe_Center, defaultValue=<rgba 0 15 180 1>}
  n8 = NamedRerouteUsage()
  n9 = LinearInterpolate(A=Constant(), B=Constant() {r=350}, Alpha=@n8) {constA=20, constB=15}
  n10 = Subtract(
   A=Multiply(A=@n4.R, B=ScalarParameter() {defaultValue=0.15, group=HeavyRoar, parameterName=HR_LeadingEdge_NoiseIntensity}) {constB=0.4}
   B=SphereMask(A=@n6, B=@n7.RGB, Radius=@n9) {attenuationRadius=100, hardnessPercent=50}
  )
  n11 = SmoothStep(Value=Distance(A=OneMinus(None=@n10), B=Constant() {r=0.5})) {constMin=0.8}
  n12 = Multiply(
   A=Subtract(A=@n11, B=SphereMask(A=@n6, B=@n7.RGB, Radius=@n9) {attenuationRadius=100, hardnessPercent=5})
   B=ScalarParameter() {defaultValue=1.45, group=HeavyRoar, parameterName=HR_LeadingEdge_Spread}
  ) {constB=1.6}
  n13 = SphereMask(A=@n6, B=@n7.RGB, Radius=LinearInterpolate(A=Constant(), B=Constant() {r=780}, Alpha=@n8) {constA=20, constB=15}) {attenuationRadius=100, hardnessPercent=0}
  n14 = Multiply(
   A=Subtract(A=OneMinus(None=@n11), B=OneMinus(None=@n13))
   B=ScalarParameter() {defaultValue=1.54, group=HeavyRoar, parameterName=HR_LeadingEdge_Spread_2}
  ) {constB=1.6}
  n15 = ScalarParameter() {defaultValue=1, group=HeavyRoar, parameterName=MaterialCurveC}
  n16 = Power(
   Base=LinearInterpolate(A=@n12, B=@n14, Alpha=@n15)
   Exp=ScalarParameter() {defaultValue=20, group=HeavyRoar, parameterName=HR_LeadingEdge_Power}
  ) {constExponent=10}
  n17 = Multiply(
   A=Saturate(None=LinearInterpolate(A=@n16, Alpha=@n8) {constB=0})
   B=ScalarParameter() {defaultValue=0.35, group=HeavyRoar, parameterName=HR_LeadingEdge_Intensity}
  )
  n18 = Power(
   Base=LinearInterpolate(A=@n12, B=@n14, Alpha=@n15)
   Exp=ScalarParameter() {defaultValue=1, group=HeavyRoar, parameterName=HR_SparkEdge_Power}
  )
  n19 = NamedRerouteUsage()
  n20 = ScalarParameter() {group=HeavyRoar, parameterName=MaterialCurveD}
  n21 = Multiply(
   A=LinearInterpolate(A=Multiply(A=@n18, B=@n19), Alpha=OneMinus(None=@n20)) {constB=0}
   B=LinearInterpolate(
    A=ScalarParameter() {defaultValue=4, group=HeavyRoar, parameterName=HR_SparksIntensity}
    B=ScalarParameter() {defaultValue=5, group=HeavyRoar, parameterName=HR_SparksIntensity_2}
    Alpha=@n15
   ) {constB=0}
  )
  NamedRerouteDeclaration(None=Add(A=Saturate(None=@n17), B=Saturate(None=@n21))) {name=LeadingEdgeMask, nodeColor=<rgba 1 0.7031 0 1>}
  n22 = NamedRerouteUsage()
  n23 = NamedRerouteUsage()
  n24 = TextureSample(Tex=TextureObjectParameter() {group=Skin FX, parameterName=SkinFX_Mask})
  n25 = MaterialFunctionCall(In=@n24.G, Contrast=ScalarParameter() {defaultValue=0.1, group=Skin FX, parameterName=SkinFX_Mask_GreenChannelContrast}) {
   desc="@pedram.navin\r\nreduce compression artifacts"
   materialFunction=&CheapContrast
  }
  MaterialExpressionAppend4Vector_0 = MaterialXAppend4Vector!(A=@n24.R, B=@n25.Result, C=@n24.R, D=@n24.R)
  n26 = NamedRerouteDeclaration(None=@MaterialExpressionAppend4Vector_0) {name=SkinFX, nodeColor=<rgba 0.8047 1 0 1>}
  n27 = StaticComponentMaskParameter(None=@n26) {defaultR=true, group=Skin FX, parameterName=PanningEmissive_Channel}
  n28 = NamedRerouteUsage()
  n29 = Saturate(None=Add(A=NamedRerouteDeclaration(None=@n27) {name=PanningEmissiveMask, nodeColor=<rgba 0 1 0.8359 1>}, B=@n28))
  n30 = ScalarParameter() {defaultValue=10, group=HeavyRoar, parameterName=HR_EdgeMaskAlphaIntensity}
  n31 = Saturate(None=Add(A=Multiply(A=@n28, B=@n30) {constB=2}, B=NamedRerouteUsage()))
  n32 = NamedRerouteUsage()
  n33 = StaticSwitchParameter(
   True=MaterialFunctionCall!(
    Alpha=LinearInterpolate(B=@n29, Alpha=@n31)
    isVehicle
    Gmap=@n23
    UVOffset=StaticComponentMaskParameter(None=@n26) {defaultB=true, group=Skin FX, parameterName=PanningEmissive_UVOffset_Channel}
    Local Normal
    Local Position
    EmissiveColorOverride
    Material=@n32
   ) {materialFunction=/Game/Weapons/WeaponSkins/MaterialFunction/PanningEmissive}.Result
   False=@n32
  ) {group=Skin FX, parameterName=Use Panning Emissive}
  n34 = StaticSwitchParameter(
   True=MaterialFunctionCall!(Alpha=StaticComponentMaskParameter(None=@n26) {defaultR=true, group=Skin FX, parameterName=Cloth Channel}, Gmap=@n23, Material=@n33) {materialFunction=/Game/Weapons/WeaponSkins/MaterialFunction/ClothFuzz}.Result
   False=@n33
  ) {group=Skin FX, parameterName=Use Cloth fuzz}
  n35 = StaticSwitchParameter(
   True=MaterialFunctionCall!(Alpha=StaticComponentMaskParameter(None=@n26) {defaultR=true, group=Skin FX, parameterName=ThinFilm_Channel}, Material Attributes=@n34) {materialFunction=/Game/Weapons/WeaponSkins/MaterialFunction/ThinFilmFunction}.Result
   False=@n34
  ) {group=Skin FX, parameterName=Use Thin Film}
  n36 = GetMaterialAttributes(None=StaticSwitchParameter(
    True=MaterialFunctionCall(Mask=@n22, Materials=@n35) {materialFunction=&MF_MetalsLUT}.Result
    False=@n35
   ) {group=Skin FX, parameterName=UseMetalLUT}) {attributeGetTypes [#1]}
  n37 = Multiply(
   A=MaterialFunctionCall(
    Hue Shift Percentage=ScalarParameter() {defaultValue=0.2, group=HeavyRoar, parameterName=HR_LeadingEdge_HueShift}
    Texture=@n36.Emissive Color
   ) {materialFunction=&HueShift}.Result
   B=ScalarParameter() {defaultValue=2, group=HeavyRoar, parameterName=HR_LeadingEdge_EmissiveIntensity}
  )
  n38 = NamedRerouteUsage()
  NamedRerouteDeclaration(None=SetMaterialAttributes!(
    MaterialAttributes=@n36.MaterialAttributes
    Emissive Color=LinearInterpolate(A=@n36.Emissive Color, B=@n37, Alpha=@n38)
    Opacity Mask
    Base Color
   ) {attributeSetTypes [#1, #2, #3]}) {name=MA_OUT, nodeColor=<rgba 1 0.6797 0 1>}
  n39 = SetMaterialAttributes!(MaterialAttributes=NamedRerouteUsage(), Base Color=NamedRerouteUsage()) {attributeSetTypes [#3]}
  n40 = NamedRerouteUsage()
  n41 = VectorParameter() {group=HeavyRoar, parameterName=New Skin Boost Color And Exponent, defaultValue=<rgba 0.87 0.6688 0.6948 2>}
  n42 = OneMinus(None=Fresnel(ExponentIn=@n41.A) {exponent=2})
  n43 = Multiply(A=@n40, B=@n41.RGB)
  n44 = Multiply(A=ScalarParameter() {defaultValue=0.75, parameterName=Global Skin Boost EX}, B=Add(A=Multiply(A=@n42, B=@n43), B=@n43))
  n45 = ScalarParameter() {defaultValue=1, group=SkinColor, parameterName=HumanSkin}
  n46 = SetMaterialAttributes!(
   MaterialAttributes=@n39
   Base Color=Saturate(None=LinearInterpolate(A=@n40, B=@n44, Alpha=@n45))
   Shading Model=ShadingModel() {shadingModel=MSM_SubsurfaceProfile}
   Opacity=ScalarParameter() {defaultValue=1, group=SkinColor, parameterName=SSS Brightness}
  ) {attributeSetTypes [#3, #4, #5]}
  n47 = NamedRerouteUsage()
  NamedRerouteDeclaration(None=BlendMaterialAttributes(A=NamedRerouteUsage(), B=BlendMaterialAttributes(A=@n39, B=@n46, Alpha=@n47), Alpha=NamedRerouteUsage())) {name=MA_Skin, nodeColor=<rgba 1 0 0.9609 1>}
  NamedRerouteDeclaration(None=TextureSample(Tex=TextureObjectParameter() {group=HeavyRoar, parameterName=HR_Diffuse2}).RGB) {name=NewBaseColor, nodeColor=<rgba 0.6406 1 0 1>}
  n48 = Add(A=TextureCoordinate() {uTiling=3, vTiling=3}, B=Multiply(A=Time(), B=Constant2Vector() {g=0.3, r=0.1}) {constB=0.1})
  n49 = Multiply(
   A=ConstantBiasScale(None=NamedRerouteUsage()) {bias=-0.5, scale=2}
   B=ScalarParameter() {defaultValue=0.1, group=HeavyRoar, parameterName=HR_NoiseWarpIntensity}
  )
  n50 = Multiply(
   A=ComponentMask(None=NamedRerouteUsage()) {g=true, r=true}
   B=ScalarParameter() {defaultValue=0.03, group=HeavyRoar, parameterName=HR_NormalWarpIntensity}
  )
  n51 = NamedRerouteUsage()
  n52 = TextureSample(UVs=Add(A=Add(A=@n48, B=@n49), B=@n50) {desc=UV Warp}, Tex=@n51) {
   samplerSource=SSM_Wrap_WorldGroupSettings
   samplerType=SAMPLERTYPE_Masks
   texture=&T_M_MED_HeavyRoar_Style_TA
  }
  n53 = Add(A=TextureCoordinate() {uTiling=4, vTiling=4}, B=Multiply(A=Time(), B=Constant2Vector() {g=0.7, r=-0.2}) {constB=0.1})
  n54 = TextureSample(UVs=Add(A=@n50, B=Add(A=@n53, B=@n49)) {desc=UV Warp}, Tex=@n51) {
   samplerSource=SSM_Wrap_WorldGroupSettings
   samplerType=SAMPLERTYPE_Masks
   texture=&T_M_MED_HeavyRoar_Style_TA
  }
  NamedRerouteDeclaration(None=LinearInterpolate(A=@n52.G, B=@n54.G, Alpha=NamedRerouteUsage())) {name=Sparks, nodeColor=<rgba 1 0.2344 0 1>}
  n55 = CollectionParameter() {collection=&/MPC_HeavyRoar, parameterName=HR_MaterialCurveB}
  n56 = StaticSwitchParameter(True=@n55, False=ScalarParameter() {group=HeavyRoar, parameterName=MaterialCurveB}) {group=HeavyRoar, parameterName=UseMPC}
  n57 = NamedRerouteDeclaration(None=StaticSwitchParameter(True=ScalarParameter() {group=HeavyRoar, parameterName=MaterialCurveA}, False=OneMinus(None=@n56)) {group=HeavyRoar, parameterName=DefaultToStyle}) {name=Progress, nodeColor=<rgba 0 1 0.4844 1>}
  NamedRerouteDeclaration(None=StaticSwitchParameter(True=@n57, False=OneMinus(None=@n57)) {group=HeavyRoar, parameterName=DefaultToStyle}) {name=SkinProgress, nodeColor=<rgba 0 0.7656 1 1>}
  NamedRerouteDeclaration(None=GetMaterialAttributes(None=NamedRerouteDeclaration(None=@n5.Result) {name=MA_IN, nodeColor=<rgba 0.5234 1 0 1>}) {attributeGetTypes [#6]}.Normal) {name=Normal_IN, nodeColor=<rgba 1 0 0.9375 1>}
  n58 = Multiply(A=Time(), B=ScalarParameter() {defaultValue=-0.5, group=HeavyRoar, parameterName=HR_DebugSpeed}) {constB=0.2}
  n59 = ScalarParameter() {group=HeavyRoar, parameterName=MaterialCurveB}
  NamedRerouteDeclaration(None=StaticSwitchParameter(True=Frac(None=@n58), False=OneMinus(None=@n59)) {group=HeavyRoar, parameterName=HR_UseDebug}) {name=Time, nodeColor=<rgba 1 0 0.1406 1>}
  NamedRerouteDeclaration(None=@n5.Skin) {name=SkinMask, nodeColor=<rgba 0.4531 0 1 1>}
  NamedRerouteDeclaration(None=@n3) {name=TA_Tex, nodeColor=<rgba 0.0781 1 0 1>}
  n60 = Multiply(A=Subtract(A=@n11) {constB=0.9}) {constB=10}
  NamedRerouteDeclaration(None=Saturate(None=LinearInterpolate(A=@n60, B=@n11, Alpha=@n15))) {name=EmissiveWipeMask, nodeColor=<rgba 0 1 0.0156 1>}
  NamedRerouteUsage()
  NamedRerouteUsage()
  SetMaterialAttributes!(MaterialAttributes, Base Color=Constant3Vector() {constant=<rgba 1 0 0 0>}) {attributeSetTypes [#3]}
  n61 = MaterialFunctionCall(
   AO=NamedRerouteDeclaration(None=@n5.AO) {name=AO, nodeColor=<rgba 1 0.1641 0 1>}
   Distance Blend=@n5.Distance Blend
   Fresnel=@n5.Distance Blend
   Gmap=NamedRerouteDeclaration(None=@n5.Gmap) {name=GMap, nodeColor=<rgba 0 1 0.5547 1>}
   Pre Skinned Local Position=@n5.Distance Blend
   Material=NamedRerouteUsage()
  ) {materialFunction=&BaseFN_CharacterFunction_PostFX}
  n62 = MaterialFunctionCall(In=@n61.Result) {materialFunction=&CINE_MF_CharPostFX}
  StaticSwitchParameter(
   True=MaterialFunctionCall!(In=@n62.Result) {materialFunction=/Game/Weapons/WeaponSkins/Wraps/MaterialFunctions/MF_3LightsMobile}.Result
   False=@n62.Result
  ) {parameterName=UseCinematicFakeLighting}
 ]
}
