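# Post-process-style animation blueprint for the HeavyRoar Style male face accessory, driving its hood and strings.
# It copies the parent mesh pose, layers in body-mesh arm bones, and runs the HeavyRoar hood Control Rig.
# Two RigidBody passes simulate hood and strings in spine_05 space.
# The emote_BuiltIn curve, read from the front-end or player anim instance, sets both hood physics alpha and DefaultSlot montage blend on hood bones.
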AnimBlueprint /BRCosmetics/Characters/Player/Male/Medium/Bodies/M_MED_HeavyRoar_Style/Meshes/Parts/M_MED_HeavyRoar_Style_FaceAcc_AnimBP
target_skeleton=&/M_MED_HeavyRoar_Style_FaceAcc_Skeleton
graphs [
 AnimGraph {nodes [
   n1 = Get BodyMesh() -> (BodyMesh:Skeletal Mesh Component Object Reference)
   n2 = Control Rig(Source=Layered blend per bone(
     BasePose=Copy Pose From Mesh() {bCopyCurves=true, bUseAttachedParent=true}
     BlendPoses_0=Copy Pose From Mesh(SourceMeshComponent=@n1) {bUseAttachedParent=true}
    ) {layerSetup [{branchFilters [{boneName=upperarm_l}, {boneName=upperarm_r}]}]}) {
    alphaBoolBlend {blendOption=Linear}
    controlRigAssetReference {blueprintRigClass=/BRCosmetics/Characters/Player/Male/Medium/Bodies/M_MED_HeavyRoar/Meshes/Parts/M_MED_HeavyRoar_Hood_CtrlRig.M_MED_HeavyRoar_Hood_CtrlRig_C}
   }
   n3 = RigidBody!(
    ComponentPose=Local To Component(LocalPose=@n2)
    Alpha=Get hoodRigidAlpha() -> (hoodRigidAlpha:double)
    OverrideWorldGravity:Vector="0,0,-980"
    ExternalForce
   ) -> (Pose) {
    bClampLinearTranslationLimitToRefPose=true
    bOverrideWorldGravity=true
    overridePhysicsAsset=&/M_MED_HeavyRoar_Style_FaceAcc_Dyn_Hood
    simulationSpace=BaseBoneSpace
    alphaBoolBlend {blendOption=Linear}
    baseBoneRef {boneName=spine_05}
    componentAppliedLinearAccClamp=<3000 1000 1000>
    componentLinearAccScale {x=0.15}
    componentLinearVelScale {x=0.2}
    overrideWorldGravity {z=-980}
    simSpaceSettings {maxAngularAcceleration=100, maxAngularVelocity=2, maxLinearAcceleration=500, maxLinearVelocity=100, velocityScaleZ=0, worldAlpha=1}
   }
   n4 = RigidBody(ComponentPose=@n3, OverrideWorldGravity="0,0,-980") {
    bClampLinearTranslationLimitToRefPose=true
    bOverrideWorldGravity=true
    overridePhysicsAsset=&/M_MED_HeavyRoar_Style_FaceAcc_Dyn_Strings
    simulationSpace=BaseBoneSpace
    alphaBoolBlend {blendOption=Linear}
    baseBoneRef {boneName=spine_05}
    componentAppliedLinearAccClamp=<3000 1000 1000>
    componentLinearAccScale {x=0.15}
    componentLinearVelScale {x=0.2}
    overrideWorldGravity {z=-980}
    simSpaceSettings {maxAngularAcceleration=100, maxAngularVelocity=2, maxLinearAcceleration=500, maxLinearVelocity=100, velocityScaleZ=0, worldAlpha=1}
   }
   n5 = Layered blend per bone(
    BasePose=Component To Local(ComponentPose=@n4)
    BlendPoses_0=Slot 'DefaultSlot'()
    BlendWeights_0=Get isEmoteHeavyRoar() -> (isEmoteHeavyRoar:double)
   ) {layerSetup [{branchFilters [{boneName=anim_hood_r}, {boneName=anim_hood_top_02}, {boneName=anim_hood_l}, {boneName=anim_hood_top_01}]}]}
   Output Pose(Result=@n5) {layerGroup=None}
  ]}
 EventGraph {nodes [
   n1 = Cast To PlayerPawn_Athena(Object=TryGetPawnOwner())
   > Set BodyMesh(BodyMesh=GetSkeletalMeshForPartType(self=@n1))
   n2 = GetAnimInstance(self=GetOwnerSkeletalMeshComponent())
   > Is Valid(InputObject=@n2)
   Event BlueprintUpdateAnimation()
   > n3 = Branch(Condition=Get bIsInFrontEnd())
   > n4 = Set isEmoteHeavyRoar(isEmoteHeavyRoar:double=GetCurveValue(self=GetOwnerFrontEndAnimInstance(), CurveName=emote_BuiltIn)) -> (Output_Get:double)
   > Set hoodRigidAlpha(hoodRigidAlpha:double=@n4) -> (Output_Get:double)
   n5 = Set isEmoteHeavyRoar(
    execute=@n3.else
    isEmoteHeavyRoar=GetCurveValue(
     self=GetOwnerFortPlayerAnimInstance!(self:Custom Character Part Anim Instance Object Reference) -> (ReturnValue:Fort Player Anim Instance Object Reference)
     CurveName=emote_BuiltIn
    )
   )
   > Set hoodRigidAlpha(hoodRigidAlpha=@n5)
   SelectFloat(A=0, B=1, bPickA=Montage_IsPlaying(self=@n2, Montage=/BRCosmetics/Animation/Game/MainPlayer/Emotes/HeavyRoar/Emote_HeavyRoar_CMM_M))
  ]}
]
variables [
 BodyMesh {type=SkeletalMeshComponent}
 hoodRigidAlpha {type=number}
 isEmoteHeavyRoar {type=number}
]
